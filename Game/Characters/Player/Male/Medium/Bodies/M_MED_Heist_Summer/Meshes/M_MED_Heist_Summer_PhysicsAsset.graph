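# Physics asset for the Heist Summer medium male player body, built for the coat's coattails and covering only those bones, the pelvis and the thighs.
# Box bodies form two-segment front, side and back chains per side with limited swing and twist, hung from tiny capsule master bodies under the pelvis.
# Pelvis and thigh capsules supply leg volume.
# Their joints and the master joints rotate freely, as does an extra left-only link between lower front and side segments.
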PhysicsAsset /Game/Characters/Player/Male/Medium/Bodies/M_MED_Heist_Summer/Meshes/M_MED_Heist_Summer_PhysicsAsset
bodies {
 Dyn_coattail_01_side_r [{shapeType=Box, extentX=13.5, extentY=3.6, extentZ=8.5, center=<5.0634 -0.0001 0.0001>, rotation=<rot 90 0 0>}]
 Dyn_coattail_back_01_l [{shapeType=Box, extentX=11.5, extentY=3.6, extentZ=7.5, center=<4.5486 0 0>, rotation=<rot 90 0 0>}]
 Dyn_coattail_back_01_r [{shapeType=Box, extentX=11.5, extentY=3.6, extentZ=7.5, center=<4.5486 0 0>, rotation=<rot 90 0 0>}]
 Dyn_coattail_back_02_l [{shapeType=Box, extentX=11.5, extentY=3.6, extentZ=7, center=<3.8712 0.0025 -0.1682>, rotation=<rot 90 0 0>}]
 Dyn_coattail_back_02_r [{shapeType=Box, extentX=11.5, extentY=3.6, extentZ=7, center=<3.8712 0.0025 -0.1682>, rotation=<rot 90 0 0>}]
 Dyn_coattail_front_01_l [{shapeType=Box, extentX=13.3708, extentY=3.5, extentZ=9.5, center=<4.7063 0 0>, rotation=<rot 90 0 0>}]
 Dyn_coattail_front_01_r [{shapeType=Box, extentX=13, extentY=3.1, extentZ=8, center=<4.128 0.0001 0>, rotation=<rot 90 0 0>}]
 Dyn_coattail_front_02_l [{shapeType=Box, extentX=15.3124, extentY=2.2788, extentZ=9.5, center=<4.6322 0 0>, rotation=<rot 90 0 0>}]
 Dyn_coattail_front_02_r [{shapeType=Box, extentX=14.5, extentY=2.7429, extentZ=8, center=<4.3966 0 0>, rotation=<rot 90 -0.0004 -0.0008>}]
 Dyn_coattail_front_mstr_l [{shapeType=Capsule, length=0.505, radius=0.505, rotation=<rot 0 -90 0>}]
 Dyn_coattail_front_mstr_r [{shapeType=Capsule, length=0.505, radius=0.505, rotation=<rot 0 -90 0>}]
 Dyn_coattail_mstr_all [{shapeType=Capsule, length=0.505, radius=0.505, rotation=<rot 0 -90 0>}]
 Dyn_coattail_side_01_l [{shapeType=Box, extentX=11.5, extentY=1.6, extentZ=7, center=<3.8086 0.0002 0>, rotation=<rot 90 0 0>}]
 Dyn_coattail_side_02_l [{shapeType=Box, extentX=11.5, extentY=2.6, extentZ=8, center=<5.2583 0.0001 0.0001>, rotation=<rot 90 0 0>}]
 Dyn_coattail_side_02_r [{shapeType=Box, extentX=13.5, extentY=2.2088, extentZ=8, center=<5.1068 0 0>, rotation=<rot 90 0 0>}]
 Dyn_coattail_side_mstr_l [{shapeType=Capsule, length=0.505, radius=0.505, rotation=<rot 0 -90 0>}]
 Dyn_coattail_side_mstr_r [{shapeType=Capsule, length=0.505, radius=0.505, rotation=<rot 0 -90 0>}]
 pelvis [
  {shapeType=Sphere, radius=3, center=<-5.8338 1.8943 6.0758>}
  {shapeType=Capsule, length=8.625, radius=8.875, center=<0.0587 -3.3259 0.2278>, rotation=<rot -0.0119 175.0768 179.1241>}
  {shapeType=Capsule, length=8.476, radius=7.4161, center=<3.2588 1.0152 0.1693>, rotation=<rot -0.0119 175.0768 179.1241>}
 ]
 thigh_l [
  {shapeType=Capsule, length=27.25, radius=7.25, center=<-13.3068 0 0>, rotation=<rot 88.7204 0 0>}
  {shapeType=Capsule, length=8.682, radius=8.4886, center=<-12.0615 0.1 0.0003>, rotation=<rot 90 -1.7899 -3.7337>}
 ]
 thigh_r [
  {shapeType=Capsule, length=29.75, radius=7.25, center=<14.4649 0.0001 0.5055>, rotation=<rot 87.5833 0 0>}
  {shapeType=Capsule, length=9.25, radius=8.5, center=<10.2939 -1.3276 2.4086>, rotation=<rot 88.7018 0 0>}
 ]
}
constraints [
 {bone1Name=Dyn_coattail_back_01_l, bone2Name=Dyn_coattail_mstr_all, swing1LimitDegrees=20, swing2LimitDegrees=15, twistLimitDegrees=5}
 {bone1Name=Dyn_coattail_back_02_l, bone2Name=Dyn_coattail_back_01_l, swing1LimitDegrees=50, swing2LimitDegrees=1, twistLimitDegrees=25}
 {bone1Name=Dyn_coattail_side_01_l, bone2Name=Dyn_coattail_side_mstr_l, swing1LimitDegrees=80, swing2LimitDegrees=25, twistLimitDegrees=20}
 {bone1Name=Dyn_coattail_side_02_l, bone2Name=Dyn_coattail_side_01_l, swing1LimitDegrees=80, swing2LimitDegrees=50, twistLimitDegrees=25}
 {
  bone1Name=Dyn_coattail_front_01_l
  bone2Name=Dyn_coattail_front_mstr_l
  swing1LimitDegrees=59.4286
  swing2LimitDegrees=25
  twistLimitDegrees=5
 }
 {
  bone1Name=Dyn_coattail_front_02_l
  bone2Name=Dyn_coattail_front_01_l
  swing1LimitDegrees=68
  swing2LimitDegrees=13.8571
  twistLimitDegrees=13
 }
 {
  bone1Name=Dyn_coattail_front_01_r
  bone2Name=Dyn_coattail_front_mstr_r
  swing1LimitDegrees=66.2857
  swing2LimitDegrees=20
  twistLimitDegrees=5
 }
 {
  bone1Name=Dyn_coattail_front_02_r
  bone2Name=Dyn_coattail_front_01_r
  swing1LimitDegrees=37.1429
  swing2LimitDegrees=25.4286
  twistLimitDegrees=11
 }
 {bone1Name=Dyn_coattail_back_01_r, bone2Name=Dyn_coattail_mstr_all, swing1LimitDegrees=20, swing2LimitDegrees=15, twistLimitDegrees=5}
 {bone1Name=Dyn_coattail_back_02_r, bone2Name=Dyn_coattail_back_01_r, swing1LimitDegrees=50, swing2LimitDegrees=1, twistLimitDegrees=25}
 {bone1Name=Dyn_coattail_01_side_r, bone2Name=Dyn_coattail_side_mstr_r, swing1LimitDegrees=80, swing2LimitDegrees=80, twistLimitDegrees=2}
 {bone1Name=Dyn_coattail_side_02_r, bone2Name=Dyn_coattail_01_side_r, swing1LimitDegrees=80, swing2LimitDegrees=80, twistLimitDegrees=5}
 {
  bone1Name=thigh_r
  bone2Name=pelvis
  swing1LimitDegrees=45
  swing1Motion=Free
  swing2LimitDegrees=45
  swing2Motion=Free
  twistLimitDegrees=45
  twistMotion=Free
 }
 {
  bone1Name=thigh_l
  bone2Name=pelvis
  swing1LimitDegrees=45
  swing1Motion=Free
  swing2LimitDegrees=45
  swing2Motion=Free
  twistLimitDegrees=45
  twistMotion=Free
 }
 {
  bone1Name=Dyn_coattail_front_mstr_l
  bone2Name=Dyn_coattail_mstr_all
  swing1LimitDegrees=45
  swing1Motion=Free
  swing2LimitDegrees=45
  swing2Motion=Free
  twistLimitDegrees=45
  twistMotion=Free
 }
 {
  bone1Name=Dyn_coattail_front_mstr_r
  bone2Name=Dyn_coattail_mstr_all
  swing1LimitDegrees=45
  swing1Motion=Free
  swing2LimitDegrees=45
  swing2Motion=Free
  twistLimitDegrees=45
  twistMotion=Free
 }
 {
  bone1Name=Dyn_coattail_side_mstr_r
  bone2Name=Dyn_coattail_mstr_all
  swing1LimitDegrees=45
  swing1Motion=Free
  swing2LimitDegrees=45
  swing2Motion=Free
  twistLimitDegrees=45
  twistMotion=Free
 }
 {
  bone1Name=Dyn_coattail_mstr_all
  bone2Name=pelvis
  swing1LimitDegrees=45
  swing1Motion=Free
  swing2LimitDegrees=45
  swing2Motion=Free
  twistLimitDegrees=45
  twistMotion=Free
 }
 {
  bone1Name=Dyn_coattail_side_mstr_l
  bone2Name=Dyn_coattail_mstr_all
  swing1LimitDegrees=45
  swing1Motion=Free
  swing2LimitDegrees=45
  swing2Motion=Free
  twistLimitDegrees=45
  twistMotion=Free
 }
 {
  bone1Name=Dyn_coattail_front_02_l
  bone2Name=Dyn_coattail_side_02_l
  swing1LimitDegrees=45
  swing1Motion=Free
  swing2LimitDegrees=45
  swing2Motion=Free
  twistLimitDegrees=45
  twistMotion=Free
 }
]
