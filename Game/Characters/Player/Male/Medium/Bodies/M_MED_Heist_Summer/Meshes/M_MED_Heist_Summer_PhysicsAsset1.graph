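# Physics asset for the male medium Heist Summer player body, set up for dynamic coattail motion.
# Box bodies cover front, back and side tail bones, while capsules on pelvis and thigh twist bones approximate hips and upper legs.
# First segments constrain to the pelvis, with second segments chained below on both fronts and the right side.
# Limits are tight at the back, loose at the front, and asymmetric on the sides, where the left is single-segment and stiffer.
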PhysicsAsset /Game/Characters/Player/Male/Medium/Bodies/M_MED_Heist_Summer/Meshes/M_MED_Heist_Summer_PhysicsAsset1
bodies {
 Dyn_coattail_01_side_r [{shapeType=Box, extentX=13.5, extentY=1.1, extentZ=6.5, center=<3.7916 0 0>, rotation=<rot 90 0 0>}]
 Dyn_coattail_back_01_l [{shapeType=Box, extentX=6, extentY=1.1, extentZ=6, center=<3.9492 0.003 0.2071>, rotation=<rot 90 0 0>}]
 Dyn_coattail_back_01_r [{shapeType=Box, extentX=6, extentY=1.1, extentZ=6, center=<3.9546 0 0>, rotation=<rot 90 0 0>}]
 Dyn_coattail_front_01_l [{shapeType=Box, extentX=11, extentY=0.6, extentZ=8.5, center=<4.6812 0.0001 0>, rotation=<rot 90 0 0>}]
 Dyn_coattail_front_01_r [{shapeType=Box, extentX=11, extentY=0.6, extentZ=8.5, center=<4.6812 0.0001 0>, rotation=<rot 90 0 0>}]
 Dyn_coattail_front_02_l [{shapeType=Box, extentX=10, extentY=2.5, extentZ=10, center=<6.2621 0 0>, rotation=<rot 90 0 0>}]
 Dyn_coattail_front_02_r [{shapeType=Box, extentX=10, extentY=2.5, extentZ=10, center=<6.2621 0 0>, rotation=<rot 90 0 0>}]
 Dyn_coattail_side_01_l [{shapeType=Box, extentX=11.5, extentY=1.1, extentZ=14, center=<7.2027 0.0002 0>, rotation=<rot 90 0 0>}]
 Dyn_coattail_side_02_r [{shapeType=Box, extentX=13.6308, extentY=2.5, extentZ=8.3749, center=<4.7097 0 0>, rotation=<rot 90 0 0>}]
 pelvis [
  {shapeType=Capsule, length=11.375, radius=9.625, center=<-3.5081 -0.1226 0>}
  {shapeType=Capsule, length=8.125, radius=9.125, center=<0.7215 2.4067 0>}
 ]
 thigh_twist_01_l [{shapeType=Capsule, length=19.625, radius=8, center=<2.0416 1.1358 0.3039>, rotation=<rot 90 -0.0049 -0.0049>}]
 thigh_twist_01_r [{shapeType=Capsule, length=15.25, radius=8.625, center=<5.3402 -0.1945 -0.4324>, rotation=<rot 90 -0.0197 -0.0191>}]
}
constraints [
 {bone1Name=Dyn_coattail_back_01_l, bone2Name=pelvis, swing1LimitDegrees=20, swing2LimitDegrees=15, twistLimitDegrees=5}
 {bone1Name=Dyn_coattail_side_01_l, bone2Name=pelvis, swing1LimitDegrees=8, swing2LimitDegrees=5, twistLimitDegrees=2}
 {bone1Name=Dyn_coattail_front_01_l, bone2Name=pelvis, swing1LimitDegrees=80, swing2LimitDegrees=80, twistLimitDegrees=25}
 {bone1Name=Dyn_coattail_front_01_r, bone2Name=pelvis, swing1LimitDegrees=80, swing2LimitDegrees=80, twistLimitDegrees=25}
 {bone1Name=Dyn_coattail_back_01_r, bone2Name=pelvis, swing1LimitDegrees=20, swing2LimitDegrees=15, twistLimitDegrees=5}
 {bone1Name=Dyn_coattail_01_side_r, bone2Name=pelvis, swing1LimitDegrees=80, swing2LimitDegrees=80, twistLimitDegrees=2}
 {
  bone1Name=Dyn_coattail_front_02_r
  bone2Name=Dyn_coattail_front_01_r
  swing1LimitDegrees=80
  swing2LimitDegrees=57.7143
  twistLimitDegrees=25
 }
 {bone1Name=Dyn_coattail_side_02_r, bone2Name=Dyn_coattail_01_side_r, swing1LimitDegrees=80, swing2LimitDegrees=80, twistLimitDegrees=25}
 {bone1Name=Dyn_coattail_front_02_l, bone2Name=Dyn_coattail_front_01_l, swing1LimitDegrees=80, swing2LimitDegrees=80, twistLimitDegrees=25}
]
