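# Post-copy Anim Blueprint adding secondary coattail and shorts motion to the Heist Summer male medium player body.
# Copies the parent mesh pose, shrinks the front-left coattail when crouched, then runs RigidBody simulation, adding a softened second pass while skydiving.
# AnimDynamics springs then move shorts bones, followed by corrective offsets driven by knee bend computed in EventGraph.
# Several older branches, including crouch-running and thigh constraints, never reach the output.
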AnimBlueprint /Game/Characters/Player/Male/Medium/Bodies/M_MED_Heist_Summer/Meshes/M_MED_Heist_Summer_Skeleton_AnimBP
target_skeleton=&/M_MED_Heist_Summer_Skeleton
graphs [
 AnimGraph {nodes [
   n1 = Anim Dynamics - Bone: Dyn_short_outer_r(Alpha=1, ExternalForce="0,0,0") {
    angularDampingOverride=1
    angularSpringConstant=20
    bAngularSpring=true
    bGravityOverrideInSimSpace=true
    bOverrideAngularDamping=true
    gravityScale=1.1
    simulationSpace=RootRelative
    alphaBoolBlend {blendOption=Linear}
    boundBone {boneName=Dyn_short_outer_r}
    componentAppliedLinearAccClamp=<2000 2000 2000>
    componentLinearAccScale=<0.4 0.4 0.4>
    componentLinearVelScale=<0.2 0.2 0.2>
    physicsBodyDefinitions [{
      sphereCollisionRadius=0
      boundBone {boneName=Dyn_short_outer_r}
      boxExtents=<15 4 20>
      constraintSetup {angularLimitsMax=<10 10 10>, angularLimitsMin=<-10 -10 -10>, angularTarget=<1 0 0>}
      localJointOffset=<7.5 0 -2.5>
     }]
    relativeSpaceBone {boneName=thigh_r}
   }
   n2 = Anim Dynamics - Bone: Dyn_short_front_r(ComponentPose=@n1, Alpha=1, ExternalForce="0,0,0") {
    angularDampingOverride=1
    angularSpringConstant=20
    bAngularSpring=true
    bGravityOverrideInSimSpace=true
    bOverrideAngularDamping=true
    gravityScale=1.1
    simulationSpace=RootRelative
    alphaBoolBlend {blendOption=Linear}
    boundBone {boneName=Dyn_short_front_r}
    componentAppliedLinearAccClamp=<2000 2000 2000>
    componentLinearAccScale=<0.4 0.4 0.4>
    componentLinearVelScale=<0.2 0.2 0.2>
    physicsBodyDefinitions [{
      sphereCollisionRadius=0
      boundBone {boneName=Dyn_short_front_r}
      boxExtents=<15 4 20>
      constraintSetup {angularLimitsMax=<10 10 10>, angularLimitsMin=<-10 -10 -10>, angularTarget=<1 0 0>}
      localJointOffset=<7.5 0 -2.5>
     }]
    relativeSpaceBone {boneName=thigh_r}
   }
   n3 = Anim Dynamics - Bone: Dyn_short_inner_r(ComponentPose=@n2, Alpha=1, ExternalForce="0,0,0") {
    angularDampingOverride=1
    angularSpringConstant=20
    bAngularSpring=true
    bGravityOverrideInSimSpace=true
    bOverrideAngularDamping=true
    gravityScale=1.1
    simulationSpace=RootRelative
    alphaBoolBlend {blendOption=Linear}
    boundBone {boneName=Dyn_short_inner_r}
    componentAppliedLinearAccClamp=<2000 2000 2000>
    componentLinearAccScale=<0.4 0.4 0.4>
    componentLinearVelScale=<0.2 0.2 0.2>
    physicsBodyDefinitions [{
      sphereCollisionRadius=0
      boundBone {boneName=Dyn_short_inner_r}
      boxExtents=<15 4 20>
      constraintSetup {angularLimitsMax=<10 10 10>, angularLimitsMin=<-10 -10 -10>, angularTarget=<1 0 0>}
      localJointOffset=<7.5 0 -2.5>
     }]
    relativeSpaceBone {boneName=thigh_r}
   }
   n4 = Anim Dynamics - Bone: Dyn_short_outer_l(ComponentPose=@n3, Alpha=1, ExternalForce="0,0,0") {
    angularDampingOverride=1
    angularSpringConstant=20
    bAngularSpring=true
    bGravityOverrideInSimSpace=true
    bOverrideAngularDamping=true
    gravityScale=1.1
    simulationSpace=RootRelative
    alphaBoolBlend {blendOption=Linear}
    boundBone {boneName=Dyn_short_outer_l}
    componentAppliedLinearAccClamp=<2000 2000 2000>
    componentLinearAccScale=<0.4 0.4 0.4>
    componentLinearVelScale=<0.2 0.2 0.2>
    physicsBodyDefinitions [{
      sphereCollisionRadius=0
      boundBone {boneName=Dyn_short_outer_l}
      boxExtents=<15 4 20>
      constraintSetup {angularLimitsMax=<10 10 10>, angularLimitsMin=<-10 -10 -10>, angularTarget=<1 0 0>}
      localJointOffset=<7.5 0 -2.5>
     }]
    relativeSpaceBone {boneName=thigh_l}
   }
   n5 = Anim Dynamics - Bone: Dyn_short_front_l(ComponentPose=@n4, Alpha=1, ExternalForce="0,0,0") {
    angularDampingOverride=1
    angularSpringConstant=20
    bAngularSpring=true
    bGravityOverrideInSimSpace=true
    bOverrideAngularDamping=true
    gravityScale=1.1
    simulationSpace=RootRelative
    alphaBoolBlend {blendOption=Linear}
    boundBone {boneName=Dyn_short_front_l}
    componentAppliedLinearAccClamp=<2000 2000 2000>
    componentLinearAccScale=<0.4 0.4 0.4>
    componentLinearVelScale=<0.2 0.2 0.2>
    physicsBodyDefinitions [{
      sphereCollisionRadius=0
      boundBone {boneName=Dyn_short_front_l}
      boxExtents=<15 4 20>
      constraintSetup {angularLimitsMax=<10 10 10>, angularLimitsMin=<-10 -10 -10>, angularTarget=<1 0 0>}
      localJointOffset=<7.5 0 -2.5>
     }]
    relativeSpaceBone {boneName=thigh_l}
   }
   n6 = Anim Dynamics - Bone: Dyn_short_inner_l(ComponentPose=@n5, Alpha=1, ExternalForce="0,0,0") {
    angularDampingOverride=1
    angularSpringConstant=20
    bAngularSpring=true
    bGravityOverrideInSimSpace=true
    bOverrideAngularDamping=true
    gravityScale=1.1
    simulationSpace=RootRelative
    alphaBoolBlend {blendOption=Linear}
    boundBone {boneName=Dyn_short_inner_l}
    componentAppliedLinearAccClamp=<2000 2000 2000>
    componentLinearAccScale=<0.4 0.4 0.4>
    componentLinearVelScale=<0.2 0.2 0.2>
    physicsBodyDefinitions [{
      sphereCollisionRadius=0
      boundBone {boneName=Dyn_short_inner_l}
      boxExtents=<15 4 20>
      constraintSetup {angularLimitsMax=<10 10 10>, angularLimitsMin=<-10 -10 -10>, angularTarget=<1 0 0>}
      localJointOffset=<7.5 0 -2.5>
     }]
    relativeSpaceBone {boneName=thigh_l}
   }
   n7 = Anim Dynamics - Bone: Dyn_short_back_r(ComponentPose=@n6, Alpha=1, ExternalForce="0,0,0") {
    angularDampingOverride=1
    angularSpringConstant=20
    bAngularSpring=true
    bGravityOverrideInSimSpace=true
    bOverrideAngularDamping=true
    gravityScale=1.1
    simulationSpace=RootRelative
    alphaBoolBlend {blendOption=Linear}
    boundBone {boneName=Dyn_short_back_r}
    componentAppliedLinearAccClamp=<2000 2000 2000>
    componentLinearAccScale=<0.4 0.4 0.4>
    componentLinearVelScale=<0.2 0.2 0.2>
    physicsBodyDefinitions [{
      sphereCollisionRadius=0
      boundBone {boneName=Dyn_short_back_r}
      boxExtents=<15 4 20>
      constraintSetup {angularLimitsMax=<10 10 10>, angularLimitsMin=<-10 -10 -10>, angularTarget=<1 0 0>}
      localJointOffset=<7.5 0 -2.5>
     }]
    relativeSpaceBone {boneName=thigh_l}
   }
   Anim Dynamics - Bone: Dyn_short_back_l(ComponentPose=@n7, Alpha=1, ExternalForce="0,0,0") {
    angularDampingOverride=1
    angularSpringConstant=20
    bAngularSpring=true
    bGravityOverrideInSimSpace=true
    bOverrideAngularDamping=true
    gravityScale=1.1
    simulationSpace=RootRelative
    alphaBoolBlend {blendOption=Linear}
    boundBone {boneName=Dyn_short_back_l}
    componentAppliedLinearAccClamp=<2000 2000 2000>
    componentLinearAccScale=<0.4 0.4 0.4>
    componentLinearVelScale=<0.2 0.2 0.2>
    physicsBodyDefinitions [{
      sphereCollisionRadius=0
      boundBone {boneName=Dyn_short_back_l}
      boxExtents=<15 4 20>
      constraintSetup {angularLimitsMax=<10 10 10>, angularLimitsMin=<-10 -10 -10>, angularTarget=<1 0 0>}
      localJointOffset=<7.5 0 -2.5>
     }]
    relativeSpaceBone {boneName=thigh_l}
   }
   n8 = "Transform (Modify) Bone - Bone: Dyn_coattail_mstr_all"(ComponentPose=Local To Component(LocalPose=Use cached pose 'Rigids'()), Alpha=1, Translation="0,1,2", Rotation="0,10,0", Scale="1,1,1") {
    rotationSpace=BCS_BoneSpace
    translationMode=BMM_Additive
    alphaBoolBlend {blendOption=Linear}
    boneToModify {boneName=Dyn_coattail_mstr_all}
    rotation {yaw=10}
    translation {y=1, z=2}
   }
   n9 = "Transform (Modify) Bone - Bone: Dyn_coattail_front_mstr_r"(ComponentPose=@n8, Alpha=1, Translation="5,5,0", Rotation="0,10,0", Scale="1,1,1") {
    rotationMode=BMM_Additive
    rotationSpace=BCS_BoneSpace
    translationMode=BMM_Additive
    translationSpace=BCS_BoneSpace
    alphaBoolBlend {blendOption=Linear}
    boneToModify {boneName=Dyn_coattail_front_mstr_r}
    rotation {yaw=10}
    translation=<5 5>
   }
   n10 = "Transform (Modify) Bone - Bone: Dyn_coattail_front_mstr_l"(ComponentPose=@n9, Alpha=1, Translation="-5,3,0", Rotation="0,10,0", Scale="1,1,1") {
    rotationMode=BMM_Additive
    rotationSpace=BCS_BoneSpace
    translationMode=BMM_Additive
    translationSpace=BCS_BoneSpace
    alphaBoolBlend {blendOption=Linear}
    boneToModify {boneName=Dyn_coattail_front_mstr_l}
    rotation {yaw=10}
    translation=<-5 3>
   }
   n11 = "Transform (Modify) Bone - Bone: Dyn_coattail_front_02_r"(ComponentPose=@n10, Alpha=1, Translation="0,5,0", Rotation="0,10,0", Scale="1,1,1") {
    rotationSpace=BCS_BoneSpace
    translationMode=BMM_Additive
    translationSpace=BCS_BoneSpace
    alphaBoolBlend {blendOption=Linear}
    boneToModify {boneName=Dyn_coattail_front_02_r}
    rotation {yaw=10}
    translation {y=5}
   }
   n12 = "Transform (Modify) Bone - Bone: Dyn_coattail_front_02_l"(ComponentPose=@n11, Alpha=1, Translation="0,3,0", Rotation="0,10,0", Scale="1,1,1") {
    rotationSpace=BCS_BoneSpace
    translationMode=BMM_Additive
    translationSpace=BCS_BoneSpace
    alphaBoolBlend {blendOption=Linear}
    boneToModify {boneName=Dyn_coattail_front_02_l}
    rotation {yaw=10}
    translation {y=3}
   }
   Blend Poses by bool(BlendPose_0=Component To Local(ComponentPose=@n12), BlendPose_1=Use cached pose 'Rigids'()) {blendType=Linear}
   n13 = "Constraint - Dyn_coattail_front_mstr_r (R)"(ComponentPose=Local To Component(LocalPose=Use cached pose 'CopyPose'()), Alpha=1, ConstraintWeights_0=1) {
    alphaBoolBlend {blendOption=Linear}
    boneToModify {boneName=Dyn_coattail_front_mstr_r}
    constraintSetup [{transformType=Rotation, targetBone {boneName=thigh_r}}]
    constraintWeights [1]
   }
   n14 = "Constraint - Dyn_coattail_side_mstr_r (R)"(ComponentPose=@n13, Alpha=0.5, ConstraintWeights_0=1) {
    alphaBoolBlend {blendOption=Linear}
    boneToModify {boneName=Dyn_coattail_side_mstr_r}
    constraintSetup [{transformType=Rotation, targetBone {boneName=thigh_r}}]
    constraintWeights [1]
   }
   n15 = "Constraint - Dyn_coattail_front_mstr_l (R)"(ComponentPose=@n14, Alpha=1, ConstraintWeights_0=1) {
    alphaBoolBlend {blendOption=Linear}
    boneToModify {boneName=Dyn_coattail_front_mstr_l}
    constraintSetup [{transformType=Rotation, targetBone {boneName=thigh_r}}]
    constraintWeights [1]
   }
   n16 = "Constraint - Dyn_coattail_side_mstr_l (R)"(ComponentPose=@n15, Alpha=0.5, ConstraintWeights_0=1) {
    alphaBoolBlend {blendOption=Linear}
    boneToModify {boneName=Dyn_coattail_side_mstr_l}
    constraintSetup [{transformType=Rotation, targetBone {boneName=thigh_r}}]
    constraintWeights [1]
   }
   Component To Local(ComponentPose=@n16)
   "Constraint - Dyn_coattail_front_r (P)"(Alpha=0.5, ConstraintWeights_0=1) {
    alphaBoolBlend {blendOption=Linear}
    boneToModify {boneName=Dyn_coattail_front_r}
    constraintSetup [{targetBone {boneName=thigh_r}}]
    constraintWeights [1]
   }
   "Constraint - Dyn_coattail_side_l (R)"(Alpha=1, ConstraintWeights_0=1) {
    alphaBoolBlend {blendOption=Linear}
    boneToModify {boneName=Dyn_coattail_side_l}
    constraintSetup [{transformType=Rotation, perAxis {bX=false, bY=false}, targetBone {boneName=Dyn_coattail_front_l}}]
    constraintWeights [1]
   }
   n17 = RigidBody(ComponentPose=Local To Component(LocalPose=Use cached pose 'CopyPose'())) {
    bClampLinearTranslationLimitToRefPose=true
    bOverrideWorldGravity=true
    overridePhysicsAsset=&/M_MED_Heist_Summer_PhysicsAsset
    simulationSpace=BaseBoneSpace
    alphaBoolBlend {blendOption=Linear}
    baseBoneRef {boneName=spine_02}
    componentLinearAccScale=<0.1 0.1 0.1>
    componentLinearVelScale=<0.1 0.1 0.1>
    overrideWorldGravity {z=-980}
   }
   Layered blend per bone(BlendPoses_0=Component To Local(ComponentPose=@n17)) {layerSetup [{branchFilters [{blendDepth=1, boneName=Dyn_coattail_front_r}, {blendDepth=1, boneName=Dyn_coattail_back_r}, {blendDepth=1, boneName=Dyn_coattail_side_r}]}]}
   "Transform (Modify) Bone - Bone: Dyn_coattail_front_l"!(
    ComponentPose
    Alpha=1
    Translation="0,0,0"
    Rotation_Roll:float=0
    Rotation_Pitch:float=0
    Rotation_Yaw:float=Get SideCoat_Compensate() -> (SideCoat_Compensate:double)
    Scale="1,1,1"
   ) -> (Pose) {rotationMode=BMM_Additive, rotationSpace=BCS_WorldSpace, alphaBoolBlend {blendOption=Linear}, boneToModify {boneName=Dyn_coattail_front_l}}
   n18 = Get calf_rot_back_l() -> (calf_rot_back_l:double)
   "Transform (Modify) Bone - Bone: dyn_shorts_main_l"(Alpha=@n18, Translation="0.000,-0.000,-4", Rotation="0,0,10", Scale="1,1,1") {
    rotationMode=BMM_Additive
    rotationSpace=BCS_BoneSpace
    scaleMode=BMM_Additive
    scaleSpace=BCS_BoneSpace
    translationMode=BMM_Additive
    translationSpace=BCS_BoneSpace
    alphaBoolBlend {blendOption=Linear}
    alphaScaleBiasClamp {bInterpResult=true, interpSpeedDecreasing=25, interpSpeedIncreasing=25}
    boneToModify {boneName=dyn_shorts_main_l}
    rotation {roll=10}
    translation {z=-4}
   }
   n19 = "Transform (Modify) Bone - Bone: Dyn_coattail_side_mstr_l"(Alpha=1, Translation="0,0,0", Rotation="5,-7.5,0", Scale="1,1,1") {
    rotationMode=BMM_Additive
    alphaBoolBlend {blendOption=Linear}
    boneToModify {boneName=Dyn_coattail_side_mstr_l}
    rotation {pitch=5, yaw=-7.5}
   }
   n20 = "Transform (Modify) Bone - Bone: Dyn_coattail_front_mstr_l"(ComponentPose=@n19, Alpha=1, Translation="5,0,5", Rotation="24,-10,20", Scale="1,1,1") {
    rotationMode=BMM_Additive
    translationMode=BMM_Additive
    alphaBoolBlend {blendOption=Linear}
    boneToModify {boneName=Dyn_coattail_front_mstr_l}
    rotation=<rot 24 -10 20>
    translation {x=5, z=5}
   }
   "Transform (Modify) Bone - Bone: Dyn_coattail_side_mstr_r"(ComponentPose=@n20, Alpha=1, Translation="0,0,0", Rotation="8,17.5,20", Scale="1,1,1") {rotationMode=BMM_Additive, alphaBoolBlend {blendOption=Linear}, boneToModify {boneName=Dyn_coattail_side_mstr_r}, rotation=<rot 8 17.5 20>}
   n21 = Get calf_rot_back_r() -> (calf_rot_back_r:double)
   "Transform (Modify) Bone - Bone: dyn_shorts_main_r"(Alpha=@n21, Translation="0.000,-0.000,3", Rotation="0,0,10", Scale="1,1,1") {
    rotationMode=BMM_Additive
    rotationSpace=BCS_BoneSpace
    scaleMode=BMM_Additive
    scaleSpace=BCS_BoneSpace
    translationMode=BMM_Additive
    translationSpace=BCS_BoneSpace
    alphaBoolBlend {blendOption=Linear}
    alphaScaleBiasClamp {bInterpResult=true, interpSpeedDecreasing=25, interpSpeedIncreasing=25}
    boneToModify {boneName=dyn_shorts_main_r}
    rotation {roll=10}
    translation {z=3}
   }
   RigidBody(ExternalForce="0,200,0") {
    bClampLinearTranslationLimitToRefPose=true
    bOverrideWorldGravity=true
    overridePhysicsAsset=&M_MED_Heist_Summer_PhysicsAsset
    simulationSpace=BaseBoneSpace
    alphaBoolBlend {blendOption=Linear}
    baseBoneRef {boneName=spine_02}
    componentLinearAccScale=<0.1 0.1 0.1>
    componentLinearVelScale=<0.1 0.1 0.1>
    overrideWorldGravity {z=-980}
   }
   n22 = Anim Dynamics - Bone: Dyn_short_inner_l(ComponentPose=Local To Component(LocalPose=Use cached pose 'RigidBodies'()), Alpha=1, ExternalForce="0,0,0") {
    angularDampingOverride=0.999
    angularSpringConstant=100
    bAngularSpring=true
    bGravityOverrideInSimSpace=true
    bOverrideAngularDamping=true
    numSolverIterationsPostUpdate=4
    numSolverIterationsPreUpdate=16
    simulationSpace=RootRelative
    alphaBoolBlend {blendOption=Linear}
    boundBone {boneName=Dyn_short_inner_l}
    physicsBodyDefinitions [{
      sphereCollisionRadius=0
      boundBone {boneName=Dyn_short_inner_l}
      boxExtents=<25 12.5 30>
      constraintSetup {angularTargetAxis=AxisZ, twistAxis=AxisZ, angularLimitsMax=<10 10 5>, angularLimitsMin=<-10 -10 -5>, angularTarget=<0 0 1>}
      localJointOffset=<3 2 -3.5>
     }]
   }
   n23 = Anim Dynamics - Bone: Dyn_short_outer_l(ComponentPose=@n22, Alpha=1, ExternalForce="0,0,0") {
    angularDampingOverride=0.999
    angularSpringConstant=100
    bAngularSpring=true
    bGravityOverrideInSimSpace=true
    bOverrideAngularDamping=true
    numSolverIterationsPostUpdate=4
    numSolverIterationsPreUpdate=16
    simulationSpace=RootRelative
    alphaBoolBlend {blendOption=Linear}
    boundBone {boneName=Dyn_short_outer_l}
    physicsBodyDefinitions [{
      sphereCollisionRadius=0
      boundBone {boneName=Dyn_short_outer_l}
      boxExtents=<50 50 30>
      constraintSetup {angularTargetAxis=AxisZ, twistAxis=AxisZ, angularLimitsMax=<10 10 5>, angularLimitsMin=<-10 -10 -5>, angularTarget=<0 0 1>}
      localJointOffset=<0 0 -6.5>
     }]
   }
   n24 = Anim Dynamics - Bone: Dyn_short_front_l(ComponentPose=@n23, Alpha=1, ExternalForce="0,0,0") {
    angularDampingOverride=0.5
    angularSpringConstant=300
    bAngularSpring=true
    bGravityOverrideInSimSpace=true
    bOverrideAngularDamping=true
    numSolverIterationsPostUpdate=4
    numSolverIterationsPreUpdate=16
    simulationSpace=RootRelative
    alphaBoolBlend {blendOption=Linear}
    boundBone {boneName=Dyn_short_front_l}
    componentAppliedLinearAccClamp=<3000 3000 3000>
    componentLinearAccScale=<0.3 0.3 0.3>
    componentLinearVelScale=<0.3 0.3 0.3>
    physicsBodyDefinitions [{
      sphereCollisionRadius=0
      boundBone {boneName=Dyn_short_front_l}
      boxExtents=<18 8 25>
      constraintSetup {angularTargetAxis=AxisZ, twistAxis=AxisZ, angularLimitsMax=<10 20 5>, angularLimitsMin=<-10 -20 -5>, angularTarget=<0 0 1>}
      localJointOffset=<4 0 2>
     }]
   }
   n25 = Anim Dynamics - Bone: Dyn_short_back_l(ComponentPose=@n24, Alpha=1, ExternalForce="0,0,0") {
    angularDampingOverride=0.999
    angularSpringConstant=300
    bAngularSpring=true
    bGravityOverrideInSimSpace=true
    bOverrideAngularDamping=true
    numSolverIterationsPostUpdate=4
    numSolverIterationsPreUpdate=16
    simulationSpace=RootRelative
    alphaBoolBlend {blendOption=Linear}
    boundBone {boneName=Dyn_short_back_l}
    physicsBodyDefinitions [{
      sphereCollisionRadius=0
      boundBone {boneName=Dyn_short_back_l}
      boxExtents=<40 38 15>
      constraintSetup {angularTargetAxis=AxisZ, twistAxis=AxisZ, angularLimitsMax=<10 0 5>, angularLimitsMin=<-10 -20 -5>, angularTarget=<0 0 1>}
      localJointOffset=<0 0 -3.5>
     }]
   }
   n26 = Anim Dynamics - Bone: Dyn_short_inner_r(ComponentPose=@n25, Alpha=1, ExternalForce="0,0,0") {
    angularDampingOverride=0.999
    angularSpringConstant=100
    bAngularSpring=true
    bGravityOverrideInSimSpace=true
    bOverrideAngularDamping=true
    numSolverIterationsPostUpdate=4
    numSolverIterationsPreUpdate=16
    simulationSpace=RootRelative
    alphaBoolBlend {blendOption=Linear}
    boundBone {boneName=Dyn_short_inner_r}
    physicsBodyDefinitions [{
      sphereCollisionRadius=0
      boundBone {boneName=Dyn_short_inner_r}
      boxExtents=<25 12.5 30>
      constraintSetup {angularTargetAxis=AxisZ, twistAxis=AxisZ, angularLimitsMax=<10 10 5>, angularLimitsMin=<-10 -10 -5>, angularTarget=<0 0 1>}
      localJointOffset=<3 5 0>
     }]
   }
   n27 = Anim Dynamics - Bone: Dyn_short_outer_r(ComponentPose=@n26, Alpha=1, ExternalForce="0,0,0") {
    angularDampingOverride=0.999
    angularSpringConstant=100
    bAngularSpring=true
    bGravityOverrideInSimSpace=true
    bOverrideAngularDamping=true
    numSolverIterationsPostUpdate=4
    numSolverIterationsPreUpdate=16
    simulationSpace=RootRelative
    alphaBoolBlend {blendOption=Linear}
    boundBone {boneName=Dyn_short_outer_r}
    physicsBodyDefinitions [{
      sphereCollisionRadius=0
      boundBone {boneName=Dyn_short_outer_r}
      boxExtents=<50 50 30>
      constraintSetup {angularTargetAxis=AxisZ, twistAxis=AxisZ, angularLimitsMax=<10 10 5>, angularLimitsMin=<-10 -10 -5>, angularTarget=<0 0 1>}
      localJointOffset=<0 0 6.5>
     }]
   }
   n28 = Anim Dynamics - Bone: Dyn_short_front_r(ComponentPose=@n27, Alpha=1, ExternalForce="0,0,0") {
    angularDampingOverride=0.5
    angularSpringConstant=300
    bAngularSpring=true
    bGravityOverrideInSimSpace=true
    bOverrideAngularDamping=true
    numSolverIterationsPostUpdate=4
    numSolverIterationsPreUpdate=16
    simulationSpace=RootRelative
    alphaBoolBlend {blendOption=Linear}
    boundBone {boneName=Dyn_short_front_r}
    componentAppliedLinearAccClamp=<3000 3000 3000>
    componentLinearAccScale=<0.3 0.3 0.3>
    componentLinearVelScale=<0.3 0.3 0.3>
    physicsBodyDefinitions [{
      sphereCollisionRadius=0
      boundBone {boneName=Dyn_short_front_r}
      boxExtents=<18 4 25>
      constraintSetup {angularTargetAxis=AxisZ, twistAxis=AxisZ, angularLimitsMax=<10 20 5>, angularLimitsMin=<-10 -20 -5>, angularTarget=<0 0 1>}
      localJointOffset=<6 -2 4>
     }]
   }
   n29 = Anim Dynamics - Bone: Dyn_short_back_r(ComponentPose=@n28, Alpha=1, ExternalForce="0,0,0") {
    angularDampingOverride=0.999
    angularSpringConstant=300
    bAngularSpring=true
    bGravityOverrideInSimSpace=true
    bOverrideAngularDamping=true
    numSolverIterationsPostUpdate=4
    numSolverIterationsPreUpdate=16
    simulationSpace=RootRelative
    alphaBoolBlend {blendOption=Linear}
    boundBone {boneName=Dyn_short_back_r}
    physicsBodyDefinitions [{
      sphereCollisionRadius=0
      boundBone {boneName=Dyn_short_back_r}
      boxExtents=<40 36 15>
      constraintSetup {angularTargetAxis=AxisZ, twistAxis=AxisZ, angularLimitsMax=<0 0 5>, angularLimitsMin=<0 -20 -5>, angularTarget=<0 0 1>}
      localJointOffset=<0 0 3.5>
     }]
   }
   n30 = "Transform (Modify) Bone - Bone: Dyn_short_back_l"(ComponentPose=@n29, Alpha=@n18, Translation="-23,0,2", Rotation="-3,0,0", Scale="0.9,0.9,0.9") {
    rotationMode=BMM_Additive
    rotationSpace=BCS_BoneSpace
    scaleMode=BMM_Additive
    scaleSpace=BCS_BoneSpace
    translationMode=BMM_Additive
    translationSpace=BCS_BoneSpace
    alphaBoolBlend {blendOption=Linear}
    alphaScaleBiasClamp {bInterpResult=true, interpSpeedDecreasing=100, interpSpeedIncreasing=100}
    boneToModify {boneName=Dyn_short_back_l}
    rotation {pitch=-3}
    scale=<0.9 0.9 0.9>
    translation {x=-23, z=2}
   }
   n31 = "Transform (Modify) Bone - Bone: Dyn_short_outer_l"(ComponentPose=@n30, Alpha=@n18, Translation="-4,2,0", Rotation="8,0.000,0", Scale="1,1,1") {
    rotationMode=BMM_Additive
    rotationSpace=BCS_BoneSpace
    scaleMode=BMM_Additive
    scaleSpace=BCS_BoneSpace
    translationMode=BMM_Additive
    translationSpace=BCS_BoneSpace
    alphaBoolBlend {blendOption=Linear}
    alphaScaleBiasClamp {bInterpResult=true, interpSpeedDecreasing=100, interpSpeedIncreasing=100}
    boneToModify {boneName=Dyn_short_outer_l}
    rotation {pitch=8}
    translation=<-4 2>
   }
   n32 = "Transform (Modify) Bone - Bone: Dyn_short_inner_l"(ComponentPose=@n31, Alpha=@n18, Translation="4,2,0", Rotation="8,-12,0.000", Scale="1,1,1") {
    rotationMode=BMM_Additive
    rotationSpace=BCS_BoneSpace
    scaleMode=BMM_Additive
    scaleSpace=BCS_BoneSpace
    translationMode=BMM_Additive
    translationSpace=BCS_BoneSpace
    alphaBoolBlend {blendOption=Linear}
    alphaScaleBiasClamp {bInterpResult=true, interpSpeedDecreasing=100, interpSpeedIncreasing=100}
    boneToModify {boneName=Dyn_short_inner_l}
    rotation {pitch=8, yaw=-12}
    translation=<4 2>
   }
   n33 = "Transform (Modify) Bone - Bone: Dyn_short_front_l"(ComponentPose=@n32, Alpha=@n18, Translation="6,0,2.5", Rotation="-13,0,0", Scale="1,1,1") {
    rotationMode=BMM_Additive
    rotationSpace=BCS_BoneSpace
    scaleMode=BMM_Additive
    scaleSpace=BCS_BoneSpace
    translationMode=BMM_Additive
    translationSpace=BCS_BoneSpace
    alphaBoolBlend {blendOption=Linear}
    alphaScaleBiasClamp {bInterpResult=true, interpSpeedDecreasing=100, interpSpeedIncreasing=100}
    boneToModify {boneName=Dyn_short_front_l}
    rotation {pitch=-13}
    translation {x=6, z=2.5}
   }
   n34 = "Transform (Modify) Bone - Bone: Dyn_short_back_r"(ComponentPose=@n33, Alpha=@n21, Translation="-20,-5,0", Rotation="-3,-24.73,-11.925", Scale="0.9,0.9,0.9") {
    rotationMode=BMM_Additive
    rotationSpace=BCS_BoneSpace
    scaleMode=BMM_Additive
    scaleSpace=BCS_BoneSpace
    translationMode=BMM_Additive
    translationSpace=BCS_BoneSpace
    alphaBoolBlend {blendOption=Linear}
    alphaScaleBiasClamp {bInterpResult=true, interpSpeedDecreasing=100, interpSpeedIncreasing=100}
    boneToModify {boneName=Dyn_short_back_r}
    rotation=<rot -3 -24.73 -11.925>
    scale=<0.9 0.9 0.9>
    translation=<-20 -5>
   }
   n35 = "Transform (Modify) Bone - Bone: Dyn_short_outer_r"(ComponentPose=@n34, Alpha=@n21, Translation="-5,0.022,-7", Rotation="8,0.000,0", Scale="1,1,1") {
    rotationMode=BMM_Additive
    rotationSpace=BCS_BoneSpace
    scaleMode=BMM_Additive
    scaleSpace=BCS_BoneSpace
    translationMode=BMM_Additive
    translationSpace=BCS_BoneSpace
    alphaBoolBlend {blendOption=Linear}
    alphaScaleBiasClamp {bInterpResult=true, interpSpeedDecreasing=100, interpSpeedIncreasing=100}
    boneToModify {boneName=Dyn_short_outer_r}
    rotation {pitch=8}
    translation=<-5 0.022 -7>
   }
   n36 = "Transform (Modify) Bone - Bone: Dyn_short_inner_r"(ComponentPose=@n35, Alpha=@n21, Translation="0,-2,0", Rotation="0,0,0", Scale="1,1,1") {
    rotationMode=BMM_Additive
    rotationSpace=BCS_BoneSpace
    scaleMode=BMM_Additive
    scaleSpace=BCS_BoneSpace
    translationMode=BMM_Additive
    translationSpace=BCS_BoneSpace
    alphaBoolBlend {blendOption=Linear}
    alphaScaleBiasClamp {bInterpResult=true, interpSpeedDecreasing=100, interpSpeedIncreasing=100}
    boneToModify {boneName=Dyn_short_inner_r}
    translation {y=-2}
   }
   n37 = "Transform (Modify) Bone - Bone: Dyn_short_front_r"(ComponentPose=@n36, Alpha=@n21, Translation="4,0.400,0.109", Rotation="-8,0,0", Scale="1,1,1") {
    rotationMode=BMM_Additive
    rotationSpace=BCS_BoneSpace
    scaleMode=BMM_Additive
    scaleSpace=BCS_BoneSpace
    translationMode=BMM_Additive
    translationSpace=BCS_BoneSpace
    alphaBoolBlend {blendOption=Linear}
    alphaScaleBiasClamp {bInterpResult=true, interpSpeedDecreasing=100, interpSpeedIncreasing=100}
    boneToModify {boneName=Dyn_short_front_r}
    rotation {pitch=-8}
    translation=<4 0.4 0.109>
   }
   Output Pose(Result=Component To Local(ComponentPose=@n37)) {layerGroup=None}
   Save cached pose 'CopyPose'(Pose=Copy Pose From Mesh() {bUseAttachedParent=true}) {cacheName=CopyPose}
   n38 = "Constraint - Dyn_coattail_front_mstr_l (R)"(ComponentPose=Local To Component(LocalPose=Use cached pose 'CopyPose'()), Alpha=1, ConstraintWeights_0=1) {
    alphaBoolBlend {blendOption=Linear}
    boneToModify {boneName=Dyn_coattail_front_mstr_l}
    constraintSetup [{transformType=Rotation, targetBone {boneName=thigh_l}}]
    constraintWeights [1]
   }
   n39 = "Constraint - Dyn_coattail_front_mstr_l (R)"(ComponentPose=@n38, Alpha=1, ConstraintWeights_0=1) {
    alphaBoolBlend {blendOption=Linear}
    boneToModify {boneName=Dyn_coattail_front_mstr_l}
    constraintSetup [{transformType=Rotation, targetBone {boneName=thigh_l}}]
    constraintWeights [1]
   }
   n40 = "Constraint - Dyn_coattail_side_mstr_l (R)"(ComponentPose=@n39, Alpha=0.4, ConstraintWeights_0=1) {
    alphaBoolBlend {blendOption=Linear}
    boneToModify {boneName=Dyn_coattail_side_mstr_l}
    constraintSetup [{transformType=Rotation, targetBone {boneName=thigh_l}}]
    constraintWeights [1]
   }
   n41 = "Constraint - Dyn_coattail_side_mstr_l (P)"(ComponentPose=@n40, Alpha=0.4, ConstraintWeights_0=1) {
    alphaBoolBlend {blendOption=Linear}
    boneToModify {boneName=Dyn_coattail_side_mstr_l}
    constraintSetup [{targetBone {boneName=thigh_l}}]
    constraintWeights [1]
   }
   n42 = "Constraint - Dyn_coattail_front_mstr_l (P)"(ComponentPose=@n41, Alpha=0.3, ConstraintWeights_0=1) {
    alphaBoolBlend {blendOption=Linear}
    boneToModify {boneName=Dyn_coattail_front_mstr_l}
    constraintSetup [{targetBone {boneName=thigh_l}}]
    constraintWeights [1]
   }
   n43 = "Constraint - Dyn_coattail_front_mstr_r (R)"(ComponentPose=@n42, Alpha=1, ConstraintWeights_0=1) {
    alphaBoolBlend {blendOption=Linear}
    boneToModify {boneName=Dyn_coattail_front_mstr_r}
    constraintSetup [{transformType=Rotation, targetBone {boneName=thigh_r}}]
    constraintWeights [1]
   }
   n44 = "Constraint - Dyn_coattail_side_mstr_r (R)"(ComponentPose=@n43, Alpha=0.5, ConstraintWeights_0=1) {
    alphaBoolBlend {blendOption=Linear}
    boneToModify {boneName=Dyn_coattail_side_mstr_r}
    constraintSetup [{transformType=Rotation, targetBone {boneName=thigh_r}}]
    constraintWeights [1]
   }
   n45 = "Transform (Modify) Bone - Bone: Dyn_coattail_side_mstr_l"(ComponentPose=@n44, Alpha=1, Translation="0,0,3", Rotation="5,-5,0", Scale="1,1,1") {
    rotationMode=BMM_Additive
    alphaBoolBlend {blendOption=Linear}
    boneToModify {boneName=Dyn_coattail_side_mstr_l}
    rotation {pitch=5, yaw=-5}
    translation {z=3}
   }
   n46 = "Transform (Modify) Bone - Bone: Dyn_coattail_front_mstr_l"(ComponentPose=@n45, Alpha=1, Translation="5,0,3", Rotation="22,0,0", Scale="1,1,1") {
    rotationMode=BMM_Additive
    alphaBoolBlend {blendOption=Linear}
    boneToModify {boneName=Dyn_coattail_front_mstr_l}
    rotation {pitch=22}
    translation {x=5, z=3}
   }
   n47 = "Transform (Modify) Bone - Bone: Dyn_coattail_side_mstr_r"(ComponentPose=@n46, Alpha=1, Translation="0,0,3", Rotation="-8,0,8", Scale="1,1,1") {
    rotationMode=BMM_Additive
    alphaBoolBlend {blendOption=Linear}
    boneToModify {boneName=Dyn_coattail_side_mstr_r}
    rotation {pitch=-8, roll=8}
    translation {z=3}
   }
   n48 = Anim Dynamics - Bone: Dyn_coattail_front_02_r(ComponentPose=@n47, Alpha=1, ExternalForce="0,0,0") {
    bGravityOverrideInSimSpace=true
    alphaBoolBlend {blendOption=Linear}
    boundBone {boneName=Dyn_coattail_front_02_r}
    componentAppliedLinearAccClamp=<3000 3000 3000>
    componentLinearAccScale=<0.3 0.3 0.3>
    componentLinearVelScale=<0.3 0.3 0.3>
    physicsBodyDefinitions [{
      sphereCollisionRadius=0
      boundBone {boneName=Dyn_coattail_front_02_r}
      boxExtents=<15 1 15>
      constraintSetup {angularLimitsMax=<10 10 10>, angularLimitsMin=<-10 -10 -10>, angularTarget=<0 0 0>}
      localJointOffset=<7 -2 0>
     }]
   }
   n49 = Anim Dynamics - Bone: Dyn_coattail_side_02_l(ComponentPose=@n48, Alpha=1, ExternalForce="0,0,0") {
    bGravityOverrideInSimSpace=true
    alphaBoolBlend {blendOption=Linear}
    boundBone {boneName=Dyn_coattail_side_02_l}
    physicsBodyDefinitions [{
      sphereCollisionRadius=0
      boundBone {boneName=Dyn_coattail_side_02_l}
      boxExtents=<8 1 8>
      constraintSetup {angularLimitsMax=<5 5 5>, angularLimitsMin=<-5 -5 -5>, angularTarget=<0 0 0>}
      localJointOffset=<6 -1 0>
     }]
   }
   n50 = Anim Dynamics - Bone: Dyn_coattail_side_02_r(ComponentPose=@n49, Alpha=1, ExternalForce="0,0,0") {
    bGravityOverrideInSimSpace=true
    alphaBoolBlend {blendOption=Linear}
    boundBone {boneName=Dyn_coattail_side_02_r}
    physicsBodyDefinitions [{
      sphereCollisionRadius=0
      boundBone {boneName=Dyn_coattail_side_02_r}
      boxExtents=<15 1 15>
      constraintSetup {angularLimitsMax=<10 10 10>, angularLimitsMin=<-10 -10 -10>, angularTarget=<0 0 0>}
      localJointOffset=<6 -1 0>
     }]
   }
   n51 = Anim Dynamics - Bone: Dyn_coattail_back_02_l(ComponentPose=@n50, Alpha=1, ExternalForce="0,0,0") {
    bGravityOverrideInSimSpace=true
    alphaBoolBlend {blendOption=Linear}
    boundBone {boneName=Dyn_coattail_back_02_l}
    componentAppliedLinearAccClamp=<3000 3000 3000>
    componentLinearAccScale=<0.3 0.3 0.3>
    componentLinearVelScale=<0.3 0.3 0.3>
    physicsBodyDefinitions [{
      sphereCollisionRadius=0
      boundBone {boneName=Dyn_coattail_back_02_l}
      boxExtents=<8 1 8>
      constraintSetup {angularLimitsMax=<5 5 5>, angularLimitsMin=<-5 -5 -5>, angularTarget=<0 0 0>}
      localJointOffset=<6 -1 0>
     }]
   }
   Save cached pose 'If Crouching'(Pose=Blend Poses by bool(BlendPose_0=Component To Local(ComponentPose=@n51), BlendPose_1=Use cached pose 'RigidBodies'(), BlendTime_1=4) {blendType=Linear}) {cacheName=If Crouching}
   n52 = RigidBody(ComponentPose=Local To Component(LocalPose=Use cached pose 'CopyPose'())) {
    bClampLinearTranslationLimitToRefPose=true
    bOverrideWorldGravity=true
    overridePhysicsAsset=&/M_MED_Heist_Summer_PhysicsAsset1
    simulationSpace=BaseBoneSpace
    alphaBoolBlend {blendOption=Linear}
    baseBoneRef {boneName=spine_02}
    componentAppliedLinearAccClamp=<2000 2000 2000>
    componentLinearAccScale=<0.005 0.005 0.005>
    componentLinearVelScale=<0.005 0.005 0.005>
    overrideWorldGravity {z=-980}
   }
   n53 = "Transform (Modify) Bone - Bone: Dyn_coattail_front_mstr_r"(ComponentPose=@n52, Alpha=1, Translation="0,3,3", Rotation="0,15,0", Scale="1,1,1") {
    rotationMode=BMM_Additive
    translationMode=BMM_Additive
    alphaBoolBlend {blendOption=Linear}
    boneToModify {boneName=Dyn_coattail_front_mstr_r}
    rotation {yaw=15}
    translation {y=3, z=3}
   }
   n54 = "Transform (Modify) Bone - Bone: Dyn_coattail_side_mstr_r"(ComponentPose=@n53, Alpha=1, Translation="0,0,0", Rotation="0,0,20", Scale="1,1,1") {
    rotationMode=BMM_Additive
    rotationSpace=BCS_BoneSpace
    translationMode=BMM_Additive
    alphaBoolBlend {blendOption=Linear}
    boneToModify {boneName=Dyn_coattail_side_mstr_r}
    rotation {roll=20}
   }
   Save cached pose 'If Crouch Running'(Pose=Blend Poses by bool(BlendPose_0=Component To Local(ComponentPose=@n54), BlendPose_1=Use cached pose 'If Crouching'()) {blendType=Linear}) {cacheName=If Crouch Running}
   n55 = RigidBody(ComponentPose=Local To Component(LocalPose=Use cached pose 'CopyPose'())) {
    bClampLinearTranslationLimitToRefPose=true
    bOverrideWorldGravity=true
    overridePhysicsAsset=&M_MED_Heist_Summer_PhysicsAsset
    simulationSpace=BaseBoneSpace
    alphaBoolBlend {blendOption=Linear}
    baseBoneRef {boneName=spine_02}
    componentLinearAccScale=<0.1 0.1 0.1>
    componentLinearVelScale=<0.1 0.1 0.1>
    overrideWorldGravity {z=-980}
   }
   Save cached pose 'Rigids'(Pose=Component To Local(ComponentPose=@n55)) {cacheName=Rigids}
   n56 = RigidBody(ComponentPose=Local To Component(LocalPose=Use cached pose 'BeforeSkydiving'())) {
    bClampLinearTranslationLimitToRefPose=true
    bOverrideWorldGravity=true
    overridePhysicsAsset=&M_MED_Heist_Summer_PhysicsAsset
    alphaBoolBlend {blendOption=Linear}
    baseBoneRef {boneName=pelvis}
    componentAppliedLinearAccClamp=<3000 3000 3000>
    componentLinearAccScale=<0.2 0.2 0.2>
    componentLinearVelScale=<0.2 0.2 0.2>
    overrideWorldGravity {z=-486}
   }
   n57 = Component To Local(ComponentPose=@n56)
   n58 = Use cached pose 'CopyPose'()
   Save cached pose 'RigidBodies'(Pose=Blend Poses by bool(
     BlendPose_0=Layered blend per bone(BasePose=@n57, BlendPoses_0=@n58, BlendWeights_0=0.6) {layerSetup [{branchFilters [{blendDepth=3, boneName=Dyn_coattail_mstr_all}]}]}
     BlendPose_1=Use cached pose 'BeforeSkydiving'()
     BlendTime_0=0.4
     BlendTime_1=0.8
     bActiveValue=Get SkydiveFix() -> (SkydiveFix:bool)
    ) {blendType=Linear}) {cacheName=RigidBodies}
   n59 = "Transform (Modify) Bone - Bone: Dyn_coattail_front_mstr_l"(
    ComponentPose=Local To Component(LocalPose=Use cached pose 'CopyPose'())
    Alpha=Get crouching alpha() -> (crouching alpha:double)
    Translation="2,2,9"
    Rotation="0,0,0"
    Scale="0.5,0.5,0.5"
   ) {
    scaleMode=BMM_Replace
    translationMode=BMM_Additive
    alphaBoolBlend {blendOption=Linear}
    alphaScaleBiasClamp {bInterpResult=true, interpSpeedDecreasing=15, interpSpeedIncreasing=15}
    boneToModify {boneName=Dyn_coattail_front_mstr_l}
    scale=<0.5 0.5 0.5>
    translation=<2 2 9>
   }
   n60 = RigidBody(ComponentPose=@n59) {
    bClampLinearTranslationLimitToRefPose=true
    bOverrideWorldGravity=true
    overridePhysicsAsset=&M_MED_Heist_Summer_PhysicsAsset
    simulationSpace=BaseBoneSpace
    alphaBoolBlend {blendOption=Linear}
    baseBoneRef {boneName=spine_01}
    componentAppliedLinearAccClamp=<3000 3000 3000>
    componentLinearAccScale=<0.1 0.1 0.1>
    componentLinearVelScale=<0.1 0.1 0.1>
    overrideWorldGravity {z=-980}
   }
   Save cached pose 'BeforeSkydiving'(Pose=Component To Local(ComponentPose=@n60)) {cacheName=BeforeSkydiving}
   Use cached pose 'RigidBodies'()
   Use cached pose 'RigidBodies'()
   Use cached pose 'If Crouch Running'()
   Get crouching alpha()
   Get NewVar_0() -> (NewVar_0:bool)
   Get IsRunning() -> (IsRunning:bool)
   Get bIsCrouching()
   Get IfCrouchRunning() -> (IfCrouchRunning:bool)
   Get NewVar_0()
  ]}
 EventGraph {nodes [
   Event OnBeginCrouching()
   > Set crouching alpha(crouching alpha:double=1) -> (Output_Get:double)
   > Set Is Crouching(Is Crouching:bool=true) -> (Output_Get:bool)
   Event OnEndCrouching()
   > Set crouching alpha(crouching alpha=0)
   > Set Is Crouching(Is Crouching=false)
   Event BlueprintUpdateAnimation()
   > n1 = Sequence!(execute) -> (then_0, then_1, then_2:Exec, then_3:Exec, then_4:Exec, then_5:Exec, then_6:Exec, then_7:Exec)
   > n2 = Branch(Condition=Get bIsJumping())
   Set jump_alpha(execute=[@n2.then, @n3.then], jump_alpha:double=1) -> (Output_Get:double)
   n3 = Branch(execute=@n2.else, Condition=Get bIsFalling())
   Set jump_alpha(execute=@n3.else, jump_alpha=0)
   n4 = Branch(execute=@n1.then_2, Condition="float >= float"(A=Get PawnSpeed2D(), B=450) {is=title K2Node_CallFunction})
   > Set IsRunning(IsRunning:bool=true) -> (Output_Get:bool)
   Set IsRunning(execute=@n4.else, IsRunning=false)
   n5 = Get PawnSpeed2D()
   n6 = Branch(execute=@n1.then_3, Condition=AND Boolean(A=Get Is Crouching() -> (Is Crouching:bool), B="float >= float"(A=@n5, B=1)))
   > Set IfCrouchRunning(IfCrouchRunning:bool=true) -> (Output_Get:bool)
   Set IfCrouchRunning(execute=@n6.else, IfCrouchRunning=false)
   n7 = GetOwnerSkeletalMeshComponent()
   Set Thigh_L_Rot(
    execute=@n1.then_6
    Thigh_L_Rot:Rotator=GetSocketRotation!(self:Scene Component Object Reference=@n7, InSocketName:Name=thigh_l) -> (ReturnValue:Rotator)
   ) -> (Output_Get:Rotator)
   > Set Thigh_R_Rot(Thigh_R_Rot:Rotator=GetSocketRotation(self=@n7, InSocketName=thigh_r)) -> (Output_Get:Rotator)
   n8 = Branch(execute=@n1.then_7, Condition=OR Boolean(A=Get bIsSkydiveDiveMode(), B=Get bIsSkydiveFloating()))
   > Set SkydiveFix(SkydiveFix:bool=true) -> (Output_Get:bool)
   Set SkydiveFix(execute=@n8.else, SkydiveFix=false)
   n9 = Get bRigidBodyEnabled()
   n10 = Get CurrentLOD()
   Set dynamics_disabled(
    execute=@n1.then_1
    dynamics_disabled:bool=OR Boolean!(A="integer > integer"(A=@n10, B=1) {is=title K2Node_CallFunction}, B=NOT Boolean(A=@n9), C:bool=Get Is Crouching()) -> (ReturnValue)
   ) -> (Output_Get:bool)
   n11 = BreakTransform(InTransform=GetDeltaTransformFromRefPose(self=GetOwnerSkeletalMeshComponent(), BoneName=calf_l, BaseName=thigh_l))
   Set calf_rot_back_l(execute=@n1.then_5, calf_rot_back_l:double=MapRangeClamped(Value=BreakRotator(InRot=@n11.Rotation).Yaw, InRangeA=15, InRangeB=115)) -> (Output_Get:double)
   n12 = BreakTransform(InTransform=GetDeltaTransformFromRefPose(self=GetOwnerSkeletalMeshComponent(), BoneName=calf_r, BaseName=thigh_r))
   Set calf_rot_back_r(execute=@n1.then_4, calf_rot_back_r:double=MapRangeClamped(Value=BreakRotator(InRot=@n12.Rotation).Yaw, InRangeA=15, InRangeB=115)) -> (Output_Get:double)
   ResetDynamics(InTeleportType=TeleportPhysics)
   n13 = Get PawnSpeed2D()
   AND Boolean(A="float <= float"!(A:double=@n13, B:double=100) -> (ReturnValue:bool) {is=title K2Node_CallFunction}, B="float >= float"(A=@n13, B=4000))
  ]}
]
variables [
 calf_rot_back_l {type=number}
 calf_rot_back_r {type=number}
 crouching alpha {type=number}
 dynamics_disabled {type=boolean}
 IfCrouchRunning {type=boolean}
 Is Crouching {type=boolean}
 IsRunning {type=boolean}
 jump_alpha {type=number}
 NewVar_0 {type=boolean, default=true}
 SideCoat_Compensate {type=number}
 SkydiveFix {type=boolean}
 Thigh_L_Rot {type=Rotator}
 Thigh_R_Rot {type=Rotator}
]
