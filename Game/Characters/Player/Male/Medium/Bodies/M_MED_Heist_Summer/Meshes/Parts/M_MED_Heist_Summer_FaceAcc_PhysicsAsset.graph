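# Physics asset for the face accessory part of the medium male Heist Summer player body.
# It defines a single capsule body on the neck_02 bone, offset and rotated from the bone origin.
# It has no constraints, so there is no jointed chain of bodies to simulate.
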PhysicsAsset /Game/Characters/Player/Male/Medium/Bodies/M_MED_Heist_Summer/Meshes/Parts/M_MED_Heist_Summer_FaceAcc_PhysicsAsset
bodies {neck_02 [{shapeType=Capsule, length=11.7073, radius=8.5227, center=<8.9454 1.6728 0.0019>, rotation=<rot 0 -112.7903 -89.9652>}]}
constraints []
